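; Gameplay ability for a Battle Pass Season 26 quest that reports when the player busts through a door.
; When activated from an event, it calls the parent, then builds a BustThroughDoor custom verb message for the owning player pawn.
; It adds a tag from the event data under the DoorBusted subject and broadcasts the message.
; It then sends a gameplay event with the BustThroughDoor tag and ends immediately.
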(event Ability|EventActivateAbilityFromEvent
  (params (EventData :type Gameplay%20Event%20Data%20Structure%20%28by%20ref%29))
  (|CallParentFunction EventData)
  (bind _result (CustomVerb|AddTagsforSubject (CustomVerb|MakeCustomVerbMessage (Variables|GAVars|Data|GetOwningPlayerPawn) "(TagName=\"CustomVerb.BustThroughDoor\")") "DoorBusted" (GameplayTags|MakeGameplayTagContainerfromTag (Utilities|Struct|BreakGameplayEventData EventData))))
  (CustomVerb|BroadcastCustomVerbMessage _result)
  (Ability|SendGameplayEventtoActor 0 "(TagName=\"CustomVerb.BustThroughDoor\")")
  (Ability|EndAbility))
